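# Material function that layers a plastic shader onto incoming vehicle material attributes.
# A static switch, PlasticToBodyShader, swaps MF_Plastic_Base for MF_Vehicle_Body_Base_v2 as the layer source.
# The layer blends over the input by INPlasticMask.
# The input's Normal then overwrites both Normal and ClearCoatBottomNormal, so the layer contributes no normal detail to either.
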MaterialFunction /VehicleCosmetics/Materials/VehicleParents_Version2/Material_Functions/MF_Apply_Plastics_v2
graph {nodes [
  n1 = GetMaterialAttributes(None=FunctionInput() {inputName=MaterialAttributes, inputType=FunctionInput_MaterialAttributes}) {attributeGetTypes [#1, #2]}
  n2 = StaticSwitchParameter(
   True=MaterialFunctionCall() {materialFunction=&/MF_Vehicle_Body_Base_v2}.Result
   False=MaterialFunctionCall() {materialFunction=/VehicleCosmetics/Materials/MaterialFunctions/MF_Plastic_Base}.Result
  ) {desc=This switch will replace Plastic with the Body shader., group=SpecialFeatures, parameterName=PlasticToBodyShader, sortPriority=36}
  n3 = GetMaterialAttributes(None=@n2)
  n4 = FunctionInput() {inputName=INPlasticMask, sortPriority=1}
  FunctionOutput(None=SetMaterialAttributes!(
    MaterialAttributes=BlendMaterialAttributes(A=@n1.MaterialAttributes, B=@n3.MaterialAttributes, Alpha=@n4)
    Normal=@n1.Normal
    ClearCoatBottomNormal=@n1.Normal
   ) {attributeSetTypes [#1, #2]}) {outputName=SetMaterialAttributes}
 ]}
