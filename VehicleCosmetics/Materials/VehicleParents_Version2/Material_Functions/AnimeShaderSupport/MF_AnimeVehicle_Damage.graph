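# Material function that adds scratch damage to anime-shaded vehicle bodies, enabled by a static switch.
# A Damage scalar times the clear coat/damage mask input forms the damage amount, and a tiled normal map supplies scratch detail.
# Normal blending, anime light direction, view angle and Fresnel shade a tinted scratch color.
# The result writes only to EmissiveColor, so damage lighting is computed in the graph.
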MaterialFunction /VehicleCosmetics/Materials/VehicleParents_Version2/Material_Functions/AnimeShaderSupport/MF_AnimeVehicle_Damage
graph {nodes [
  n1 = NamedRerouteUsage()
  n2 = NamedRerouteUsage()
  FunctionOutput(None=StaticSwitchParameter(
    True=SetMaterialAttributes!(MaterialAttributes=@n1, EmissiveColor=@n2) {attributeSetTypes [#1]}
    False=NamedRerouteUsage()
   ) {defaultValue=true, group=Vehicle_Damage, parameterName=Vehicle Body Shows Damage})
  n3 = GetMaterialAttributes(None=FunctionInput() {inputType=FunctionInput_MaterialAttributes}) {attributeGetTypes [#1]}
  NamedRerouteDeclaration(None=@n3.EmissiveColor) {name=EmissiveSource, nodeColor=<rgba 1 0 0.375 1>}
  n4 = NamedRerouteUsage()
  n5 = NamedRerouteUsage()
  n6 = NamedRerouteUsage()
  n7 = VectorParameter() {group=Vehicle_Damage, parameterName=DamageScratchColor, defaultValue=<rgba 0.2 0.1666 0.1289 0>}
  n8 = Multiply(A=@n6, B=@n7.RGB)
  n9 = VectorParameter() {group=Vehicle_Damage, parameterName=DamageTextureTile, defaultValue=<rgba 1.5 1.5 0 0>}
  n10 = TextureSample(UVs=Multiply(A=TextureCoordinate() {uTiling=0.7, vTiling=0.7}, B=ComponentMask(None=@n9.RGB) {g=true, r=true})) {samplerType=SAMPLERTYPE_Normal, texture=/Game/Global/Textures/Default/Tileable_Patterns/CarDamageNormal}
  n11 = OneMinus(None=@n10.B)
  n12 = Multiply(A=Multiply(A=@n11, B=@n11)) {constB=10}
  n13 = Add(
   A=LinearInterpolate(A=Multiply(A=@n4) {constB=0.7}, Alpha=Fresnel(Normal=@n5) {exponent=2}) {constA=0.05, constB=0}
   B=LinearInterpolate(A=@n6, B=Multiply(A=@n6, B=@n8), Alpha=Clamp(None=@n12))
  )
  n14 = NamedRerouteUsage()
  n15 = Transform(None=@n14)
  n16 = DotProduct(
   A=MaterialFunctionCall() {materialFunction=/Game/Global/Materials/Character_Parent_2023/Anime_Parent_2023/MaterialFunctions/MFC_AnimeShader_LightDirection}.Direction Vector
   B=@n15
  )
  n17 = CameraVectorWS()
  n18 = Transform(None=VertexNormalWS()) {transformSourceType=TRANSFORMSOURCE_World, transformType=TRANSFORM_Tangent}
  n19 = ConstantBiasScale(None=LinearInterpolate(A=DotProduct(A=@n17, B=@n15), B=DotProduct(A=@n14, B=@n18))) {bias=-0.6, scale=3}
  n20 = LinearInterpolate(A=@n7.RGB, Alpha=Saturate(None=Add(A=@n16, B=@n19)))
  n21 = NamedRerouteUsage()
  NamedRerouteDeclaration(None=LinearInterpolate(A=NamedRerouteUsage(), B=Multiply(A=@n13, B=@n20), Alpha=Multiply(A=@n21)) {constAlpha=0.85}) {name=EmissiveUpdated, nodeColor=<rgba 0 0.6953 1 1>}
  NamedRerouteDeclaration(None=@n3.MaterialAttributes) {name=MA_Input, nodeColor=<rgba 1 0.9844 0 1>}
  n22 = NamedRerouteUsage()
  n23 = NamedRerouteUsage()
  n24 = Add(
   A=MaterialFunctionCall!(Material=@n23) {materialFunction=/Engine/Functions/MaterialLayerFunctions/MatLayerBlend_BreakNormal}.Normal
   B=NamedRerouteUsage()
  )
  NamedRerouteDeclaration(None=GetMaterialAttributes(None=MaterialFunctionCall!(NormalMask=@n22, Material=@n23, Normal=@n24) {materialFunction=/Engine/Functions/MaterialLayerFunctions/MatLayerBlend_NormalBlend}.Blended Material) {attributeGetTypes [#2]}.Normal) {name=NormalUpdated, nodeColor=<rgba 0 0.2969 1 1>}
  NamedRerouteDeclaration(None=Multiply(A=@n10.RGB, B=VectorParameter() {group=Vehicle_Damage, parameterName=DamageNormalMult, defaultValue=<rgba 1.6 1.6 0 1>}.RGB)) {name=DamageNormal, nodeColor=<rgba 0 0.3906 1 1>}
  n25 = FunctionInput() {
   bUsePreviewValueAsDefault=true
   inputName="Clear Coat/Damage Mask (_M B)"
   inputType=FunctionInput_Scalar
   sortPriority=1
   previewValue=<1 1 1 1>
  }
  n26 = Power(Base=ScalarParameter() {group=Vehicle_Damage, parameterName=Damage})
  NamedRerouteDeclaration(None=Clamp(None=Multiply(A=@n25, B=@n26))) {name=Damage_Value-Mask, nodeColor=<rgba 0 1 0.0859 1>}
 ]}
