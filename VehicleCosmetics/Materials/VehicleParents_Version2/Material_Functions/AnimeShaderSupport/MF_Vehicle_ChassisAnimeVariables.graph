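# Material function that exposes vehicle chassis parameters and masks as outputs for anime shader support.
# It passes out Blueprint-driven boost, brake, reverse and headlight scalars, chassis colors, and parts, paint and boost masks from one packed texture.
# Reverse, headlight and brake light masks come from EmissiveMasks RGB, via optional UV cropping or per-channel bump offset.
# Bump offset runs only at High and Epic quality, and is flagged for rework.
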MaterialFunction /VehicleCosmetics/Materials/VehicleParents_Version2/Material_Functions/AnimeShaderSupport/MF_Vehicle_ChassisAnimeVariables
graph {nodes [
  Comment() {text=Emissive Bump Offset}
  Comment() {
   text="REWORK NEEDED - ADDED FOR RAGE\r\n-Be sure to setup cropped emissive support\r\n-Combine bumpoffsets into one"
   commentColor=<rgba 0.4375 0 0 1>
  }
  n1 = VectorParameter() {group=Vehicle_Chassis_Emissive, parameterName=EmissiveUVs_RG_UpperLeftCorner_BA_LowerRightCorner, defaultValue=<rgba 0 0 1 1>}
  n2 = AppendVector(A=@n1.RGB, B=@n1.RGB)
  n3 = TextureObjectParameter() {
   desc=_E
   group=Vehicle_Chassis_Textures
   parameterName=EmissiveMasks
   samplerType=SAMPLERTYPE_LinearColor
   texture=/VehicleCosmetics/Textures/T_Black_Linear
  }
  Composite!(RGBA=@n2, T2D_In=@n3)
  FunctionOutput(None=NamedRerouteUsage()) {outputName=Masks_Parts, sortPriority=30}
  FunctionOutput(None=NamedRerouteUsage()) {outputName=Masks_Paint, sortPriority=31}
  FunctionOutput(None=NamedRerouteUsage()) {outputName=Color_Brake, sortPriority=20}
  FunctionOutput() {outputName=------, sortPriority=39}
  FunctionOutput() {outputName=----, sortPriority=19}
  FunctionOutput(None=NamedRerouteUsage()) {outputName=BPVariable_BoostGlowIntensity}
  FunctionOutput(None=NamedRerouteUsage()) {outputName=BPVariable_isBraking}
  FunctionOutput(None=NamedRerouteUsage()) {outputName=BPVariable_ReverseLight}
  FunctionOutput(None=NamedRerouteUsage()) {outputName=BPVariable_HeadlightsFullyOn}
  FunctionOutput(None=NamedRerouteUsage()) {outputName=Masks_Boost, sortPriority=32}
  FunctionOutput(None=NamedRerouteUsage()) {outputName=EmissiveMasks_ReverseLightMask, sortPriority=40}
  FunctionOutput(None=NamedRerouteUsage()) {outputName=EmissiveMasks_Headlight, sortPriority=41}
  FunctionOutput(None=NamedRerouteUsage()) {outputName=EmissiveMasks_BrakeLight, sortPriority=42}
  FunctionOutput() {outputName=-----, sortPriority=29}
  FunctionOutput(None=NamedRerouteUsage()) {outputName=Color_Headlight, sortPriority=20}
  FunctionOutput(None=NamedRerouteUsage()) {outputName=Color_Boost, sortPriority=20}
  FunctionOutput(None=NamedRerouteUsage()) {outputName=Color_Parts, sortPriority=20}
  n4 = VectorParameter() {group=Vehicle_Chassis_Color, parameterName=Chassis_HeadlightColor, defaultValue=<rgba 0.6249 1 0.9901 1>}
  NamedRerouteDeclaration(None=@n4.RGB) {name=HeadlightColor, nodeColor=<rgba 0.9282 0 1 1>}
  NamedRerouteDeclaration(None=VectorParameter() {group=Vehicle_Chassis_Color, parameterName=Chassis_BoostColor, defaultValue=<rgba 1 0.05 0 1>}.RGB) {name=BoostColor, nodeColor=<rgba 0.9282 0 1 1>}
  NamedRerouteDeclaration(None=ScalarParameter() {group=BP-USE-ONLY, parameterName=BoostGlowIntensity}) {name=BoostGlowVariable, nodeColor=<rgba 0 0 0 1>}
  NamedRerouteDeclaration(None=ScalarParameter() {group=BP-USE-ONLY, parameterName=isBraking}) {name=isBrakingVariable, nodeColor=<rgba 0 0 0 1>}
  NamedRerouteDeclaration(None=ScalarParameter() {group=BP-USE-ONLY, parameterName=ReverseLight}) {name=ReverseLightVariable, nodeColor=<rgba 0 0 0 1>}
  NamedRerouteDeclaration(None=ScalarParameter() {group=BP-USE-ONLY, parameterName=HeadlightsFullyOn}) {name=HeadlightsFullyOnVariable, nodeColor=<rgba 0 0 0 1>}
  n5 = TextureSampleParameter2D() {
   desc=_M
   group=Vehicle_Chassis_Textures
   parameterName=Masks
   samplerSource=SSM_Wrap_WorldGroupSettings
   samplerType=SAMPLERTYPE_LinearColor
   sortPriority=33
   texture=&T_Black_Linear
   channelNames {a="", b=BoostMask, g=PaintMask, r=EmssivePartsMask}
  }
  NamedRerouteDeclaration(None=@n5.BoostMask) {name=ChassisBoostEmissiveMask, nodeColor=<rgba 0 0.1726 0.4792 1>}
  MaterialExpressionReroute_5 = Reroute(None=@n2)
  MaterialExpressionReroute_6 = Reroute(None=@n3)
  n6 = TextureSample(Tex=@MaterialExpressionReroute_6.T2D_In) {samplerType=SAMPLERTYPE_LinearColor, texture=&T_Black_Linear}
  n7 = ComponentMask(None=@n6.RGBA) {g=true}
  n8 = ScalarParameter() {defaultValue=0.02, group=Vehicle_Chassis_Emissive, parameterName=Headlight_OffsetRatio, sortPriority=52}
  n9 = ComponentMask(None=@n6.RGBA) {b=true}
  n10 = ScalarParameter() {defaultValue=0.03, group=Vehicle_Chassis_Emissive, parameterName=Brakelight_OffsetRatio, sortPriority=54}
  n11 = Max(
   A=TextureSample(UVs=BumpOffset(Height=@n7, HeightRatioInput=@n8) {heightRatio=0.02, referencePlane=1}, Tex=@MaterialExpressionReroute_6.T2D_In) {samplerType=SAMPLERTYPE_LinearColor, texture=&T_Black_Linear}.G
   B=TextureSample(UVs=BumpOffset(Height=@n9, HeightRatioInput=@n10) {heightRatio=0.03, referencePlane=1}, Tex=@MaterialExpressionReroute_6.T2D_In) {samplerType=SAMPLERTYPE_LinearColor, texture=&T_Black_Linear}.B
  )
  n12 = TextureSample(
   UVs=BumpOffset(Height=@n6.R, HeightRatioInput=ScalarParameter() {group=Vehicle_Chassis_Emissive, parameterName=Reverselight_OffsetRatio, sortPriority=56}) {heightRatio=0.03, referencePlane=1}
   Tex=@MaterialExpressionReroute_6.T2D_In
  ) {samplerType=SAMPLERTYPE_LinearColor, texture=&T_Black_Linear}
  n13 = StaticSwitchParameter(True=Multiply(A=Max(A=@n11, B=@n12.R), B=@n6.RGBA), False=@n6.RGBA) {group=Vehicle_Chassis_Emissive, parameterName=BumpOffsetEmissive, sortPriority=50}
  n14 = StaticSwitchParameter(
   True=MaterialFunctionCall!(
    UVs
    TextureIn=@MaterialExpressionReroute_6.T2D_In
    UpperLeftCorner=ComponentMask(None=@MaterialExpressionReroute_5.RGBA) {g=true, r=true}
    LowerRightCorner=ComponentMask(None=@MaterialExpressionReroute_5.RGBA) {a=true, b=true}
    ExportFloat4
   ) {materialFunction=/Engine/Functions/Engine_MaterialFunctions01/Texturing/TextureCropping}.CroppedMasked
   False=QualitySwitch(Default=@n6.RGBA, High=@n13, Epic=@n13)
  ) {desc=CROP_EMIS, group=Vehicle_Chassis_Emissive, parameterName=CroppedEmissive}
  n15 = ComponentMask(None=@n14) {g=true}
  MaterialExpressionReroute_8 = Reroute(None=@n15)
  NamedRerouteDeclaration(None=@MaterialExpressionReroute_8.HeadLight) {name=HeadlightEmissiveMask, nodeColor=<rgba 0 0.1726 0.4792 1>}
  n16 = ComponentMask(None=@n14) {b=true}
  MaterialExpressionReroute_9 = Reroute(None=@n16)
  NamedRerouteDeclaration(None=@MaterialExpressionReroute_9.BrakeLight) {name=BrakeLightEmissiveMask, nodeColor=<rgba 0 0.1726 0.4792 1>}
  n17 = ComponentMask(None=@n14) {r=true}
  MaterialExpressionReroute_7 = Reroute(None=@n17)
  NamedRerouteDeclaration(None=@MaterialExpressionReroute_7.ReverseLight) {name=ReverseLightEmissiveMask, nodeColor=<rgba 0 0.1726 0.4792 1>}
  NamedRerouteDeclaration(None=@n5.EmssivePartsMask) {name=ChassisEmissivePartsMask, nodeColor=<rgba 0 0.1726 0.4792 1>}
  n18 = VectorParameter() {
   desc="The Glowing parts of the emissive (if any) controlled by the EmissivePartsMask. --"
   group=Vehicle_Chassis_Color
   parameterName=Chassis_EmissivePartsColor
   defaultValue=<rgba 0.6038 0.3325 0.013 1>
  }
  NamedRerouteDeclaration(None=@n18.RGB) {name=ChassisEmissivePartsColor, nodeColor=<rgba 0.9301 0 1 1>}
  NamedRerouteDeclaration(None=@n5.PaintMask) {name=ChassisPaintedMask, nodeColor=<rgba 0 0.1726 0.4792 1>}
  NamedRerouteDeclaration(None=VectorParameter() {group=Vehicle_Chassis_Color, parameterName=Brake_Color, defaultValue=<rgba 0.6 0 0 1>}.RGB) {name=ChassisBrakeColor, nodeColor=<rgba 0.9301 0 1 1>}
  n19 = VectorParameter() {
   desc="The Painted Color controlled by the PaintMask. -- Used for the paintable areas of the chassis --\r\nSetting alpha to 0 will bypass colorization and show the basecolor texture instead; useful for painted trims that need BC texture for None"
   group=Color
   parameterName=Chassis_PaintColor
   defaultValue=<rgba 0 0.1434 1 1>
  }
  NamedRerouteDeclaration(None=@n19.RGB) {name=PaintColor, nodeColor=<rgba 0.9282 0 1 1>}
  PinBase() {reroutePins [
    RGBA {expression=&self:MaterialExpressionReroute_5}
    T2D_In {expression=&self:MaterialExpressionReroute_6}
   ]}
  PinBase!(ReverseLight=@n17, HeadLight=@n15, BrakeLight=@n16) {reroutePins [
    ReverseLight {expression=&self:MaterialExpressionReroute_7}
    HeadLight {expression=&self:MaterialExpressionReroute_8}
    BrakeLight {expression=&self:MaterialExpressionReroute_9}
   ]}
 ]}
