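# Vehicle material function for anime shader support that can replace an incoming color, going by the name base color, with emissive colors.
# One shared scalar Color_Input feeds four parallel lerps through a named reroute.
# Each lerp blends toward its headlight, chassis brake, chassis emissive parts or boost color, using that part's scalar ECReplaceBC input as alpha.
# The four results are separate, unchained outputs, named with a _Bool suffix although they carry blended colors.
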MaterialFunction /VehicleCosmetics/Materials/VehicleParents_Version2/Material_Functions/AnimeShaderSupport/MF_Vehicle_ReplaceBCwithEmissive
graph {nodes [
  FunctionOutput(None=LinearInterpolate(
    A=NamedRerouteUsage()
    B=FunctionInput() {inputName=HeadlightColor, sortPriority=5}
    Alpha=FunctionInput() {inputName=Headlight_ECReplaceBC, inputType=FunctionInput_Scalar, sortPriority=6}
   )) {outputName=Headlight_Bool}
  FunctionOutput(None=LinearInterpolate(
    A=NamedRerouteUsage()
    B=FunctionInput() {inputName=ChassisBrakeColor, sortPriority=10}
    Alpha=FunctionInput() {inputName=Brake_ECReplaceBC, inputType=FunctionInput_Scalar, sortPriority=11}
   )) {outputName=Brake_Bool, sortPriority=10}
  FunctionOutput(None=LinearInterpolate(
    A=NamedRerouteUsage()
    B=FunctionInput() {inputName=ChassisEmissivePartsColor, sortPriority=15}
    Alpha=FunctionInput() {inputName=Parts_ECReplaceBC, inputType=FunctionInput_Scalar, sortPriority=16}
   )) {outputName=Parts_Bool, sortPriority=20}
  FunctionOutput(None=LinearInterpolate(
    A=NamedRerouteUsage()
    B=FunctionInput() {inputName=BoostColor, sortPriority=20}
    Alpha=FunctionInput() {inputName=Boost_ECReplaceBC, inputType=FunctionInput_Scalar, sortPriority=21}
   )) {outputName=Boost_Bool, sortPriority=30}
  NamedRerouteDeclaration(None=FunctionInput() {inputName=Color_Input, inputType=FunctionInput_Scalar}) {name=ColorInput, nodeColor=<rgba 0 1 0.5078 1>}
 ]}
