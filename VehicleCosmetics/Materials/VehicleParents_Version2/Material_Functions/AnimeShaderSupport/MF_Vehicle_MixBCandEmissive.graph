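# Material function in the vehicle anime shader support set that builds a combined EmissiveColors output for headlights, brakes, emissive chassis parts and boost.
# It passes a scalar Color_Input, plus each part's color and replace-base-color scalar, into MF_Vehicle_ReplaceBCwithEmissive.
# That function's four per-part Bool outputs are multiplied by emissive masks, routed through named reroutes, then summed.
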MaterialFunction /VehicleCosmetics/Materials/VehicleParents_Version2/Material_Functions/AnimeShaderSupport/MF_Vehicle_MixBCandEmissive
graph {nodes [
  n1 = MaterialFunctionCall!(
   Color_Input=FunctionInput() {inputName=Color_Input, inputType=FunctionInput_Scalar}
   HeadlightColor=FunctionInput() {inputName=HeadlightColor, sortPriority=5}
   Headlight_ECReplaceBC=FunctionInput() {inputName=Headlight_ECReplaceBC, inputType=FunctionInput_Scalar, sortPriority=6}
   ChassisBrakeColor=FunctionInput() {inputName=ChassisBrakeColor, sortPriority=10}
   Brake_ECReplaceBC=FunctionInput() {inputName=Brake_ECReplaceBC, inputType=FunctionInput_Scalar, sortPriority=11}
   ChassisEmissivePartsColor=FunctionInput() {inputName=ChassisEmissivePartsColor, sortPriority=15}
   Parts_ECReplaceBC=FunctionInput() {inputName=Parts_ECReplaceBC, inputType=FunctionInput_Scalar, sortPriority=16}
   BoostColor=FunctionInput() {inputName=BoostColor, sortPriority=20}
   Boost_ECReplaceBC=FunctionInput() {inputName=Boost_ECReplaceBC, inputType=FunctionInput_Scalar, sortPriority=21}
  ) {materialFunction=&/MF_Vehicle_ReplaceBCwithEmissive}
  n2 = Add(A=Multiply(A=@n1.Headlight_Bool, B=NamedRerouteUsage()), B=Multiply(A=@n1.Brake_Bool, B=NamedRerouteUsage()))
  n3 = Multiply(A=@n1.Parts_Bool, B=NamedRerouteUsage())
  n4 = NamedRerouteUsage()
  FunctionOutput(None=Add(A=Add(A=@n2, B=@n3), B=Multiply(A=@n1.Boost_Bool, B=@n4))) {outputName=EmissiveColors}
  NamedRerouteDeclaration(None=FunctionInput() {inputName=Headlight_EmissiveMask, inputType=FunctionInput_Scalar, sortPriority=25}) {name=HeadlightEmissiveMask, nodeColor=<rgba 1 0 0.7734 1>}
  NamedRerouteDeclaration(None=FunctionInput() {inputName=Brake_EmissiveMask, inputType=FunctionInput_Scalar, sortPriority=30}) {name=BrakeEmissiveMask, nodeColor=<rgba 1 0 0.7734 1>}
  NamedRerouteDeclaration(None=FunctionInput() {inputName=Boost_EmissiveMask, inputType=FunctionInput_Scalar, sortPriority=40}) {name=BoostEmissiveMask, nodeColor=<rgba 1 0 0.7734 1>}
  NamedRerouteDeclaration(None=FunctionInput() {inputName=Parts_EmissiveMask, inputType=FunctionInput_Scalar, sortPriority=35}) {name=PartsEmissiveMask, nodeColor=<rgba 1 0 0.7734 1>}
 ]}
