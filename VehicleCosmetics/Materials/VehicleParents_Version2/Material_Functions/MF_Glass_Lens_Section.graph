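# Material function that shades vehicle glass lenses, writing color, PBR, opacity, normal, thin-translucent and refraction values into material attributes.
# Fresnel blends inner-to-rim opacity and drives optional refraction; lens color takes an optional linear-burn tint.
# An optional fake inner thickness, derived from view angle against vertex normal and tangent, adjusts color and opacity.
# Static switches gate overrides; Fresnel_Strength and the loose parameter nodes feed nothing.
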MaterialFunction /VehicleCosmetics/Materials/VehicleParents_Version2/Material_Functions/MF_Glass_Lens_Section
graph {nodes [
  n1 = SetMaterialAttributes!(
   MaterialAttributes=NamedRerouteUsage()
   BaseColor=NamedRerouteUsage()
   Metallic=NamedRerouteUsage()
   Specular=NamedRerouteUsage()
   Roughness=NamedRerouteUsage()
   Opacity=NamedRerouteUsage()
   Normal=NamedRerouteUsage()
   TransmittanceColor=NamedRerouteUsage()
   ThinTranslucentSurfaceCoverage=NamedRerouteUsage()
   Refraction=NamedRerouteUsage()
  ) {attributeSetTypes [
    #1
    #2
    #3
    #4
    #5
    #6
    #7
    #8
    #9
   ]}
  FunctionOutput(None=@n1)
  n2 = FunctionInput(Preview=Constant() {r=5}) {
   bUsePreviewValueAsDefault=true
   description=Controls the power of the lens Fresnel
   inputName=Fresnel_Exponent
   inputType=FunctionInput_Scalar
   sortPriority=1
   previewValue=<5 5 5 5>
  }
  n3 = Fresnel(ExponentIn=@n2)
  n4 = FunctionInput(Preview=Constant() {r=1}) {
   bUsePreviewValueAsDefault=true
   description=Multiplied against the Fresnel
   inputName=Fresnel_Strength
   inputType=FunctionInput_Scalar
   sortPriority=2
   previewValue=<1 1 1 1>
  }
  Multiply(A=@n3, B=@n4)
  NamedRerouteDeclaration(None=FunctionInput(Preview=SetMaterialAttributes!(MaterialAttributes)) {bUsePreviewValueAsDefault=true, inputType=FunctionInput_MaterialAttributes, sortPriority=-5}) {name=Input_MA, nodeColor=<rgba 1 0 0.1875 1>}
  NamedRerouteDeclaration(None=FunctionInput(Preview=StaticBool()) {bUsePreviewValueAsDefault=true, inputName=UseAdditionalThickness, inputType=FunctionInput_StaticBool, sortPriority=4}) {name=UseThickness, nodeColor=<rgba 1 0.3516 0 1>}
  n5 = Constant() {r=1}
  n6 = GetMaterialAttributes(None=NamedRerouteUsage()) {attributeGetTypes [#7, #8]}
  NamedRerouteDeclaration(None=StaticSwitch(
    True=FunctionInput(Preview=@n5) {bUsePreviewValueAsDefault=true, inputName=ThinTranslucentCoverage, inputType=FunctionInput_Scalar, sortPriority=62}
    False=@n6.ThinTranslucentSurfaceCoverage
    Value=NamedRerouteUsage()
   )) {name=ThinTranslucentCoverage, nodeColor=<rgba 1 0 0.5156 1>}
  n7 = FunctionInput(Preview=Constant() {r=1.011}) {bUsePreviewValueAsDefault=true, inputName=RefractionValue, inputType=FunctionInput_Scalar, sortPriority=51}
  n8 = Saturate(None=@n3)
  NamedRerouteDeclaration(None=StaticSwitch(True=LinearInterpolate(B=@n7, Alpha=@n8) {constA=1}, False=Constant() {r=1}, Value=NamedRerouteUsage())) {name=Refraction, nodeColor=<rgba 0 0.8359 1 1>}
  n9 = FunctionInput(Preview=Constant() {r=0.125}) {bUsePreviewValueAsDefault=true, inputName=Roughness, inputType=FunctionInput_Scalar, sortPriority=23, previewValue=<0.125 0.125 0.125 0.125>}
  n10 = GetMaterialAttributes(None=NamedRerouteUsage()) {attributeGetTypes [#2, #3, #4]}
  n11 = FunctionInput(Preview=StaticBool() {value=true}) {bUsePreviewValueAsDefault=true, inputName=OverridePBR, inputType=FunctionInput_StaticBool, sortPriority=20}
  NamedRerouteDeclaration(None=StaticSwitch(True=@n9, False=@n10.Roughness, Value=@n11)) {name=Roughness, nodeColor=<rgba 1 0.8438 0 1>}
  n12 = VertexTangentWS()
  n13 = VertexNormalWS()
  n14 = MaterialFunctionCall(Vector 3=CrossProduct(A=CrossProduct(A=@n12, B=@n13), B=CameraVectorWS())) {materialFunction=/Engine/Functions/Engine_MaterialFunctions02/Utility/VectorLength}
  n15 = OneMinus(None=DotProduct(A=CameraVectorWS(), B=VertexNormalWS()))
  n16 = FunctionInput(Preview=Constant() {desc=Controls the width of the inner thickness width, r=0.025}) {
   bUsePreviewValueAsDefault=true
   description=Controls the width of the inner thickness width
   inputName=Thickness_Width
   inputType=FunctionInput_Scalar
   sortPriority=6
   previewValue=<0.025 0.025 0.025 1>
  }
  n17 = Subtract(A=@n15, B=@n16)
  n18 = MaterialFunctionCall(Vector 3=Multiply(A=@n15, B=Saturate(None=@n17))) {materialFunction=&VectorLength}
  n19 = Saturate(None=LinearInterpolate(A=Saturate(None=@n14.V3 Length), B=Saturate(None=@n18.V3 Length), Alpha=@n14.V3 Length) {constB=0})
  n20 = FunctionInput(Preview=Constant() {r=0.375}) {
   bUsePreviewValueAsDefault=true
   description=Controls the inner opacity of the frensel
   inputName=Fresnel_InnerTransparency
   inputType=FunctionInput_Scalar
   sortPriority=3
   previewValue=<0.375 0.375 0.375 0.375>
  }
  n21 = FunctionInput(Preview=Constant() {r=0.98}) {
   bUsePreviewValueAsDefault=true
   description=Controls the opacity of the frensel rim
   inputName=Fresnel_OuterTransparency
   inputType=FunctionInput_Scalar
   sortPriority=4
   previewValue=<0.98 0.98 0.98 0.98>
  }
  n22 = LinearInterpolate(A=@n20, B=@n21, Alpha=@n8)
  n23 = Add(A=LinearInterpolate(Alpha=@n19) {constA=0.05, constB=0}, B=@n22)
  n24 = NamedRerouteUsage()
  NamedRerouteDeclaration(None=Saturate(None=StaticSwitch(True=@n23, False=@n22, Value=@n24))) {name=Opacity, nodeColor=<rgba 0.5234 0 1 1>}
  NamedRerouteDeclaration(None=FunctionInput(Preview=StaticBool()) {bUsePreviewValueAsDefault=true, inputName=OverrideThinTranslucent, inputType=FunctionInput_StaticBool, sortPriority=60}) {name=UseThinTranslucent, nodeColor=<rgba 0.125 1 0 1>}
  NamedRerouteDeclaration(None=FunctionInput(Preview=StaticBool()) {bUsePreviewValueAsDefault=true, inputName=UseRefraction, inputType=FunctionInput_StaticBool, sortPriority=50}) {name=UseRefraction, nodeColor=<rgba 1 0 0.4453 1>}
  n25 = FunctionInput(Preview=Constant() {desc=Controls the height of the inner thickness color, r=0.5}) {
   bUsePreviewValueAsDefault=true
   description=Controls the height of the inner thickness color
   inputName=Thickness_Height
   inputType=FunctionInput_Scalar
   sortPriority=7
   previewValue=<0.5 0.5 0.5 1>
  }
  n26 = Multiply(A=@n25) {constB=0.1}
  n27 = LinearInterpolate(Alpha=@n19) {constA=1, constB=0}
  n28 = Multiply(A=Add(A=Subtract(A=@n26), B=@n27), B=@n27)
  n29 = FunctionInput(Preview=Constant() {r=1}) {
   bUsePreviewValueAsDefault=true
   description=Controls the opacity of the inner thickness
   inputName=Thickness_Opacity
   inputType=FunctionInput_Scalar
   sortPriority=5
   previewValue=<1 1 1 1>
  }
  n30 = FunctionInput(Preview=Constant3Vector() {constant=<rgba 0.0677 0.0677 0.0677 1>}) {
   bUsePreviewValueAsDefault=true
   description=Controls the opacity of the frensel rim
   inputName=Color_Lens
   previewValue=<0.0677 0.0677 0.0677 1>
  }
  n31 = FunctionInput(Preview=Constant3Vector() {constant=<rgba 0.001 0.001 0.001 1>}) {
   bUsePreviewValueAsDefault=true
   description=Blended by LinearBurn over the base color
   inputName=Color_Tint
   sortPriority=11
   previewValue=<0.001 0.001 0.001 0.001>
  }
  n32 = MaterialFunctionCall!(Blend=@n30, Base=@n31) {materialFunction=/Engine/Functions/Engine_MaterialFunctions03/Blends/Blend_LinearBurn}
  n33 = FunctionInput(Preview=Constant() {r=0.5}) {
   bUsePreviewValueAsDefault=true
   description=Controls the inner opacity of the frensel
   inputName=Tint_BlendOpacity
   inputType=FunctionInput_Scalar
   sortPriority=12
   previewValue=<0 0 0 0>
  }
  n34 = StaticSwitch(
   True=LinearInterpolate(A=@n30, B=Clamp(None=@n32.Result) {clampMode=CMODE_ClampMin}, Alpha=@n33)
   False=@n30
   Value=FunctionInput(Preview=StaticBool()) {bUsePreviewValueAsDefault=true, inputName=UseTint, inputType=FunctionInput_StaticBool, sortPriority=10}
  )
  n35 = MaterialFunctionCall(Base=@n34, Blend=Multiply(A=Clamp(None=@n28) {clampMode=CMODE_ClampMin}, B=Multiply(A=@n29) {constB=15})) {materialFunction=&Blend_Overlay}
  NamedRerouteDeclaration(None=StaticSwitch(True=Add(A=@n34, B=@n35.Result), False=@n34, Value=NamedRerouteUsage())) {name=BaseColor, nodeColor=<rgba 1 0 0.9141 1>}
  n36 = FunctionInput(Preview=Constant() {r=0.4}) {bUsePreviewValueAsDefault=true, inputName=Metallic, inputType=FunctionInput_Scalar, sortPriority=21, previewValue=<0.25 0.25 0.25 0.25>}
  NamedRerouteDeclaration(None=StaticSwitch(True=@n36, False=@n10.Metallic, Value=@n11)) {name=Metallic, nodeColor=<rgba 0 1 0.7656 1>}
  n37 = FunctionInput(Preview=Constant() {r=0.8}) {bUsePreviewValueAsDefault=true, inputName=Specular, inputType=FunctionInput_Scalar, sortPriority=22, previewValue=<0.8 0.8 0.8 0.8>}
  NamedRerouteDeclaration(None=StaticSwitch(True=@n37, False=@n10.Specular, Value=@n11)) {name=Specular, nodeColor=<rgba 1 0.4453 0 1>}
  n38 = Constant3Vector() {constant=<rgba 0.7 0.7 0.7 0>}
  NamedRerouteDeclaration(None=StaticSwitch(
    True=FunctionInput(Preview=@n38) {bUsePreviewValueAsDefault=true, inputName=ThinTranslucentColor, sortPriority=61}
    False=@n6.TransmittanceColor
    Value=NamedRerouteUsage()
   )) {name=ThinTranslucentColor, nodeColor=<rgba 0.125 0 1 1>}
  NamedRerouteDeclaration(None=StaticSwitchParameter(
    True=TextureSampleParameter2D() {group=Lenses, parameterName=Normal, samplerType=SAMPLERTYPE_Normal, texture=/Engine/EngineMaterials/FlatNormal}.RGB
    False=Constant3Vector() {constant=<rgba 0 0 1 0>}
   ) {group=Lenses, parameterName=UseNormalMap}) {name=Normal, nodeColor=<rgba 0.1953 1 0 1>}
  ScalarParameter() {defaultValue=1, group=Lenses, parameterName=Thickness_Opacity, sortPriority=5}
  ScalarParameter() {defaultValue=5, group=Lenses, parameterName=Fresnel Exponent, sortPriority=1}
  ScalarParameter() {defaultValue=0.4, group=00_SharedLenses, parameterName=SharedLensesMetallic, sortPriority=27}
  ScalarParameter() {defaultValue=1, group=Lenses, parameterName=Fresnel Strength, sortPriority=2}
  ScalarParameter() {defaultValue=0.025, desc=Controls the width of the inner thickness width, group=Lenses, parameterName=Thickness_Width, sortPriority=6}
  ScalarParameter() {defaultValue=0.8, group=00_SharedLenses, parameterName=SharedLensesSpecular, sortPriority=28}
  ScalarParameter() {defaultValue=0.125, group=00_SharedLenses, parameterName=SharedLensesRoughness, sortPriority=29}
  ScalarParameter() {defaultValue=0.5, desc=Controls the height of the inner thickness color, group=Lenses, parameterName=Thickness_Height, sortPriority=7}
  ScalarParameter() {defaultValue=0.375, group=Lenses, parameterName=Fresnel Inner Transparency, sortPriority=3}
  ScalarParameter() {defaultValue=0.98, group=Lenses, parameterName=Fresnel Outer Transparency, sortPriority=4}
  ScalarParameter() {defaultValue=0.5, group=00_SharedLenses, parameterName=SharedLensesTint_BlendAmount, sortPriority=12}
  StaticBoolParameter() {group=Lenses, parameterName=UseTint, sortPriority=10}
  VectorParameter() {group=Lenses, parameterName=Color_Lens, sortPriority=0, defaultValue=<rgba 0.0677 0.0677 0.0677 1>}
  VectorParameter() {group=Lenses, parameterName=Color_LinearBurnTint, sortPriority=11, defaultValue=<rgba 0.001 0.001 0.001 1>}
 ]}
