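# Material function that adds optional rim and surface color details to vehicle lenses by brightening the BaseColor of incoming material attributes.
# Two Fresnel-like masks come from camera vector dot scaled vertex normal: rim color fades toward camera-facing areas, while a high-power falloff confines surface glow to them.
# Each color's alpha multiplies its intensity.
# Static switches gate both layers, which are summed onto BaseColor and written back to the attributes.
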MaterialFunction /VehicleCosmetics/Materials/VehicleParents_Version2/Material_Functions/MF_Lenses_LensEnhancement_v2
graph {
 properties {description=Adds additional surface and rim color details to vehicle lenses, previewBlendMode=BLEND_Translucent}
 nodes [
  n1 = Constant4Vector() {desc=Alpha increases the intensity of the color, constant=<rgba 1 1 1 0.5>}
  n2 = FunctionInput(Preview=AppendVector(A=@n1.RGBA, B=@n1.RGBA)) {bUsePreviewValueAsDefault=true, inputName=Rim_ColorXPower, inputType=FunctionInput_Vector4, sortPriority=3}
  n3 = Multiply(A=ComponentMask(None=@n2) {b=true, g=true, r=true}, B=ComponentMask(None=@n2) {a=true})
  n4 = VertexNormalWS()
  n5 = Multiply(
   A=FunctionInput(Preview=Constant() {r=1.5}) {bUsePreviewValueAsDefault=true, inputName=RimNormalDistance, inputType=FunctionInput_Scalar, sortPriority=2}
   B=@n4
  ) {constB=1.5}
  n6 = CameraVectorWS()
  n7 = Saturate(None=Power(Base=DotProduct(A=@n5, B=@n6)) {constExponent=0.5})
  n8 = StaticBool()
  n9 = Add(
   A=NamedRerouteUsage()
   B=StaticSwitch(
    True=LinearInterpolate(A=@n3, Alpha=@n7) {constB=0}
    False=Constant()
    Value=FunctionInput(Preview=@n8) {bUsePreviewValueAsDefault=true, inputName=UseRimEnhancement, inputType=FunctionInput_StaticBool, sortPriority=1}
   )
  )
  n10 = Constant4Vector() {desc=Alpha increases the intensity of the color, constant=<rgba 1 1 1 0.5>}
  n11 = FunctionInput(Preview=AppendVector(A=@n10.RGBA, B=@n10.RGBA)) {bUsePreviewValueAsDefault=true, inputName=Surface_ColorXPower, inputType=FunctionInput_Vector4, sortPriority=7}
  n12 = ComponentMask(None=@n11) {b=true, g=true, r=true}
  n13 = ComponentMask(None=@n11) {a=true}
  n14 = Multiply(
   A=@n4
   B=FunctionInput(Preview=Constant() {r=0.45}) {bUsePreviewValueAsDefault=true, inputName=SurfaceNormalDistance, inputType=FunctionInput_Scalar, sortPriority=6}
  ) {constB=0.45}
  n15 = Power(
   Base=Multiply(A=DotProduct(A=@n6, B=@n14)) {constB=2.25}
   Exp=FunctionInput(Preview=Constant() {r=50}) {bUsePreviewValueAsDefault=true, inputName=SurfaceGlow_Power, inputType=FunctionInput_Scalar, sortPriority=8}
  ) {constExponent=50}
  n16 = StaticSwitch(
   True=LinearInterpolate(B=Multiply(A=@n12, B=@n13), Alpha=Saturate(None=@n15)) {constB=0.5}
   False=Constant()
   Value=FunctionInput(Preview=StaticBool()) {bUsePreviewValueAsDefault=true, inputName=UseSurfaceEnhancement, inputType=FunctionInput_StaticBool, sortPriority=5}
  )
  FunctionOutput(None=SetMaterialAttributes!(MaterialAttributes=NamedRerouteUsage(), BaseColor=Add(A=@n9, B=@n16)) {attributeSetTypes [#1]})
  n17 = GetMaterialAttributes(None=FunctionInput() {inputName=MA, inputType=FunctionInput_MaterialAttributes, sortPriority=-1}) {attributeGetTypes [#1]}
  NamedRerouteDeclaration(None=@n17.BaseColor) {name=BaseColor, nodeColor=<rgba 1 0.5391 0 1>}
  NamedRerouteDeclaration(None=@n17.MaterialAttributes) {name=MA, nodeColor=<rgba 0.2188 0 1 1>}
  ScalarParameter() {defaultValue=1.5, group=LensEnhancement, parameterName=RimNormalDistance, sortPriority=0}
  ScalarParameter() {defaultValue=0.45, group=LensEnhancement, parameterName=SurfaceNormalDistance, sortPriority=2}
  ScalarParameter() {defaultValue=50, group=LensEnhancement, parameterName=SurfaceGlow_Power, sortPriority=4}
  VectorParameter() {
   desc=Alpha increases the intensity of the color
   group=LensEnhancement
   parameterName=Surface_ColorXPower
   sortPriority=8
   defaultValue=<rgba 1 1 1 0.5>
  }
  VectorParameter() {
   desc=Alpha increases the intensity of the color
   group=LensEnhancement
   parameterName=Rim_ColorXPower
   sortPriority=6
   defaultValue=<rgba 1 1 1 0.5>
  }
 ]
}
