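# Vehicle lens material function that builds glass material attributes for headlight, rearlight, reflector, special and miscellaneous lenses.
# Each type runs MF_Glass_Lens_Section with its own parameter group, plus optional MF_Lenses_LensEnhancement_v2 rim and surface effects.
# Starting from the headlight, static switch inputs decide which other types are chained in, blended by UDIM-tile masks from MF_UDimmMask.
# An optional final switch passes opacity through MF_SortTranslucency before output.
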MaterialFunction /VehicleCosmetics/Materials/VehicleParents_Version2/Material_Functions/MF_Lenses_Base_v2
graph {nodes [
  n1 = VectorParameter() {group=01_Headlight, parameterName=HeadlightColor, sortPriority=0, defaultValue=<rgba 0.0677 0.0677 0.0677 1>}
  n2 = ScalarParameter() {
   defaultValue=0.025
   desc=Controls the width of the inner thickness width
   group=01_Headlight
   parameterName=HeadlightThickness_Width
   sortPriority=6
  }
  n3 = ScalarParameter() {
   defaultValue=0.5
   desc=Controls the height of the inner thickness color
   group=01_Headlight
   parameterName=HeadlightThickness_Height
   sortPriority=7
  }
  n4 = VectorParameter() {group=01_Headlight, parameterName=HeadlightTint_LinearBurnColor, sortPriority=11, defaultValue=<rgba 0.001 0.001 0.001 1>}
  n5 = VectorParameter() {group=01_Headlight, parameterName=HeadlightThinTranslucent_Color, sortPriority=41, defaultValue=<rgba 0.7 0.7 0.7 1>}
  n6 = MaterialFunctionCall!(
   In=SetMaterialAttributes!(MaterialAttributes)
   Color_Lens=@n1.RGB
   Fresnel_Exponent=ScalarParameter() {defaultValue=5, group=01_Headlight, parameterName=HeadlightFresnel_Exponent, sortPriority=1}
   Fresnel_Strength=ScalarParameter() {defaultValue=1, group=01_Headlight, parameterName=HeadlightFresnel_Strength, sortPriority=2}
   Fresnel_InnerTransparency=ScalarParameter() {defaultValue=0.375, group=01_Headlight, parameterName=HeadlightFresnel_InnerTransparency, sortPriority=3}
   Fresnel_OuterTransparency=ScalarParameter() {defaultValue=0.98, group=01_Headlight, parameterName=HeadlightFresnel_OuterTransparency, sortPriority=4}
   UseAdditionalThickness=StaticBoolParameter() {group=01_Headlight, parameterName=Headlight_UseAdditionalThickness, sortPriority=5}
   Thickness_Opacity=ScalarParameter() {defaultValue=1, group=01_Headlight, parameterName=HeadlightThickness_Opacity, sortPriority=5}
   Thickness_Width=@n2
   Thickness_Height=@n3
   UseTint=StaticBoolParameter() {group=01_Headlight, parameterName=Headlight_UseTint, sortPriority=10}
   Color_Tint=@n4.RGB
   Tint_BlendOpacity=ScalarParameter() {defaultValue=0.5, group=01_Headlight, parameterName=HeadlightTint_BlendAmount, sortPriority=12}
   OverridePBR=StaticBool() {value=true}
   Metallic=ScalarParameter() {defaultValue=0.25, group=01_Headlight, parameterName=HeadlightMetallic, sortPriority=27}
   Specular=ScalarParameter() {defaultValue=0.8, group=01_Headlight, parameterName=HeadlightSpecular, sortPriority=28}
   Roughness=ScalarParameter() {defaultValue=0.125, group=01_Headlight, parameterName=HeadlightRoughness, sortPriority=29}
   UseRefraction=StaticBoolParameter() {group=01_Headlight, parameterName=Headlight_UseRefraction, sortPriority=30}
   RefractionValue=ScalarParameter() {defaultValue=1.011, group=01_Headlight, parameterName=HeadlightRefractionValue, sortPriority=31}
   OverrideThinTranslucent=StaticBoolParameter() {group=01_Headlight, parameterName=Headlight_UseThinTranslucent, sortPriority=40}
   ThinTranslucentColor=@n5.RGB
   ThinTranslucentCoverage=ScalarParameter() {defaultValue=1, group=01_Headlight, parameterName=HeadlightThinTranslucent_Coverage, sortPriority=42}
  ) {materialFunction=&/MF_Glass_Lens_Section}
  n7 = VectorParameter() {
   desc=Alpha increases the intensity of the color
   group=01_Headlight_Enhancement
   parameterName=HeadlightLenses_RimColorXPower
   sortPriority=2
   defaultValue=<rgba 1 1 1 0.5>
  }
  n8 = VectorParameter() {
   desc=Alpha increases the intensity of the color
   group=01_Headlight_Enhancement
   parameterName=HeadlightLenses_SurfaceColorXPower
   sortPriority=5
   defaultValue=<rgba 1 1 1 0.5>
  }
  n9 = StaticSwitchParameter(
   True=MaterialFunctionCall!(
    MA=@n6.Result
    UseRimEnhancement=StaticBoolParameter() {group=01_Headlight_Enhancement, parameterName=Headlight_UseRimEnhancement, sortPriority=0}
    RimNormalDistance=ScalarParameter() {defaultValue=1.5, group=01_Headlight_Enhancement, parameterName=HeadlightLenses_RimNormalDistance, sortPriority=1}
    Rim_ColorXPower=AppendVector(A=@n7.RGB, B=@n7.RGB)
    UseSurfaceEnhancement=StaticBoolParameter() {group=01_Headlight_Enhancement, parameterName=Headlight_UseSurfaceEnhancement, sortPriority=3}
    SurfaceNormalDistance=ScalarParameter() {defaultValue=0.45, group=01_Headlight_Enhancement, parameterName=HeadlightLenses_SurfaceNormalDistance, sortPriority=4}
    Surface_ColorXPower=AppendVector(A=@n8.RGB, B=@n8.RGB)
    SurfaceGlow_Power=ScalarParameter() {defaultValue=50, group=01_Headlight_Enhancement, parameterName=HeadlightLenses_SurfaceGlowPower, sortPriority=6}
   ) {materialFunction=&/MF_Lenses_LensEnhancement_v2}.Result
   False=@n6.Result
  ) {group=01_Headlight, parameterName=Headlight_UseLensEnhancement, sortPriority=-10}
  n10 = VectorParameter() {
   desc=Alpha increases the intensity of the color
   group=05_Miscellaneous_Enhancement
   parameterName=MiscLenses_RimColorXPower
   sortPriority=2
   defaultValue=<rgba 1 1 1 0.5>
  }
  n11 = VectorParameter() {
   desc=Alpha increases the intensity of the color
   group=05_Miscellaneous_Enhancement
   parameterName=MiscLenses_SurfaceColorXPower
   sortPriority=5
   defaultValue=<rgba 1 1 1 0.5>
  }
  n12 = VectorParameter() {group=05_Miscellaneous, parameterName=MiscColor, sortPriority=0, defaultValue=<rgba 0.3438 0.3438 0.3438 1>}
  n13 = ScalarParameter() {
   defaultValue=0.025
   desc=Controls the width of the inner thickness width
   group=05_Miscellaneous
   parameterName=MiscThickness_Width
   sortPriority=6
  }
  n14 = ScalarParameter() {
   defaultValue=0.5
   desc=Controls the height of the inner thickness color
   group=05_Miscellaneous
   parameterName=MiscThickness_Height
   sortPriority=7
  }
  n15 = VectorParameter() {group=05_Miscellaneous, parameterName=MiscTint_LinearBurnColor, sortPriority=11, defaultValue=<rgba 0.001 0.001 0.001 1>}
  n16 = VectorParameter() {group=05_Miscellaneous, parameterName=MiscThinTranslucent_Color, sortPriority=41, defaultValue=<rgba 0.7 0.7 0.7 1>}
  n17 = MaterialFunctionCall(
   In=@n6.Result
   Color_Lens=@n12.RGB
   Fresnel_Exponent=ScalarParameter() {defaultValue=5, group=05_Miscellaneous, parameterName=MiscFresnel_Exponent, sortPriority=1}
   Fresnel_Strength=ScalarParameter() {defaultValue=1, group=05_Miscellaneous, parameterName=MiscFresnel_Strength, sortPriority=2}
   Fresnel_InnerTransparency=ScalarParameter() {defaultValue=0.375, group=05_Miscellaneous, parameterName=MiscFresnel_InnerTransparency, sortPriority=3}
   Fresnel_OuterTransparency=ScalarParameter() {defaultValue=0.98, group=05_Miscellaneous, parameterName=MiscFresnel_OuterTransparency, sortPriority=4}
   UseAdditionalThickness=StaticBoolParameter() {group=05_Miscellaneous, parameterName=Misc_UseAdditionalThickness, sortPriority=5}
   Thickness_Opacity=ScalarParameter() {defaultValue=1, group=05_Miscellaneous, parameterName=MiscThickness_Opacity, sortPriority=5}
   Thickness_Width=@n13
   Thickness_Height=@n14
   UseTint=StaticBoolParameter() {group=05_Miscellaneous, parameterName=Misc_UseTint, sortPriority=10}
   Color_Tint=@n15.RGB
   Tint_BlendOpacity=ScalarParameter() {defaultValue=0.5, group=05_Miscellaneous, parameterName=MiscTint_BlendAmount, sortPriority=12}
   OverridePBR=StaticBoolParameter() {group=05_Miscellaneous, parameterName=Misc_UsePBR, sortPriority=26}
   Metallic=ScalarParameter() {defaultValue=0.25, group=05_Miscellaneous, parameterName=MiscMetallic, sortPriority=27}
   Specular=ScalarParameter() {defaultValue=0.8, group=05_Miscellaneous, parameterName=MiscSpecular, sortPriority=28}
   Roughness=ScalarParameter() {defaultValue=0.125, group=05_Miscellaneous, parameterName=MiscRoughness, sortPriority=29}
   UseRefraction=StaticBoolParameter() {group=05_Miscellaneous, parameterName=Misc_UseRefraction, sortPriority=30}
   RefractionValue=ScalarParameter() {defaultValue=1.011, group=05_Miscellaneous, parameterName=MiscRefractionValue, sortPriority=31}
   OverrideThinTranslucent=StaticBoolParameter() {group=05_Miscellaneous, parameterName=Misc_UseThinTranslucent, sortPriority=40}
   ThinTranslucentColor=@n16.RGB
   ThinTranslucentCoverage=ScalarParameter() {defaultValue=1, group=05_Miscellaneous, parameterName=MiscThinTranslucent_Coverage, sortPriority=42}
  ) {materialFunction=&MF_Glass_Lens_Section}
  n18 = MaterialFunctionCall(
   MA=@n17.Result
   UseRimEnhancement=StaticBoolParameter() {group=05_Miscellaneous_Enhancement, parameterName=Misc_UseRimEnhancement, sortPriority=0}
   RimNormalDistance=ScalarParameter() {defaultValue=1.5, group=05_Miscellaneous_Enhancement, parameterName=MiscLenses_RimNormalDistance, sortPriority=1}
   Rim_ColorXPower=AppendVector(A=@n10.RGB, B=@n10.RGB)
   UseSurfaceEnhancement=StaticBoolParameter() {group=05_Miscellaneous_Enhancement, parameterName=Misc_UseSurfaceEnhancement, sortPriority=3}
   SurfaceNormalDistance=ScalarParameter() {defaultValue=0.45, group=05_Miscellaneous_Enhancement, parameterName=MiscLenses_SurfaceNormalDistance, sortPriority=4}
   Surface_ColorXPower=AppendVector(A=@n11.RGB, B=@n11.RGB)
   SurfaceGlow_Power=ScalarParameter() {defaultValue=50, group=05_Miscellaneous_Enhancement, parameterName=MiscLenses_SurfaceGlowPower, sortPriority=6}
  ) {materialFunction=&MF_Lenses_LensEnhancement_v2}
  n19 = StaticSwitch(
   True=@n9
   False=BlendMaterialAttributes(
    A=StaticSwitchParameter(True=@n18.Result, False=@n17.Result) {group=05_Miscellaneous, parameterName=Misc_UseLensEnhancement, sortPriority=-10}
    B=@n9
    Alpha=NamedRerouteUsage()
   )
   Value=NamedRerouteUsage()
  )
  n20 = VectorParameter() {group=04_Special, parameterName=SpecialColor, sortPriority=0, defaultValue=<rgba 0.0677 0.0677 0.0677 1>}
  n21 = ScalarParameter() {
   defaultValue=0.025
   desc=Controls the width of the inner thickness width
   group=04_Special
   parameterName=SpecialThickness_Width
   sortPriority=6
  }
  n22 = ScalarParameter() {
   defaultValue=0.5
   desc=Controls the height of the inner thickness color
   group=04_Special
   parameterName=SpecialThickness_Height
   sortPriority=7
  }
  n23 = VectorParameter() {group=04_Special, parameterName=SpecialTint_LinearBurnColor, sortPriority=11, defaultValue=<rgba 0.001 0.001 0.001 1>}
  n24 = VectorParameter() {group=04_Special, parameterName=SpecialThinTranslucent_Color, sortPriority=41, defaultValue=<rgba 0.7 0.7 0.7 1>}
  n25 = MaterialFunctionCall(
   In=@n19
   Color_Lens=@n20.RGB
   Fresnel_Exponent=ScalarParameter() {defaultValue=5, group=04_Special, parameterName=SpecialFresnel_Exponent, sortPriority=1}
   Fresnel_Strength=ScalarParameter() {defaultValue=1, group=04_Special, parameterName=SpecialFresnel_Strength, sortPriority=2}
   Fresnel_InnerTransparency=ScalarParameter() {defaultValue=0.375, group=04_Special, parameterName=SpecialFresnel_InnerTransparency, sortPriority=3}
   Fresnel_OuterTransparency=ScalarParameter() {defaultValue=0.98, group=04_Special, parameterName=Special_OuterTransparency, sortPriority=4}
   UseAdditionalThickness=StaticBoolParameter() {group=04_Special, parameterName=Special_UseAdditionalThickness, sortPriority=5}
   Thickness_Opacity=ScalarParameter() {defaultValue=1, group=04_Special, parameterName=SpecialThickness_Opacity, sortPriority=5}
   Thickness_Width=@n21
   Thickness_Height=@n22
   UseTint=StaticBoolParameter() {group=04_Special, parameterName=Special_UseTint, sortPriority=10}
   Color_Tint=@n23.RGB
   Tint_BlendOpacity=ScalarParameter() {defaultValue=0.5, group=04_Special, parameterName=SpecialTint_BlendAmount, sortPriority=12}
   OverridePBR=StaticBoolParameter() {group=04_Special, parameterName=Special_UsePBR, sortPriority=26}
   Metallic=ScalarParameter() {defaultValue=0.25, group=04_Special, parameterName=SpecialMetallic, sortPriority=27}
   Specular=ScalarParameter() {defaultValue=0.8, group=04_Special, parameterName=SpecialSpecular, sortPriority=28}
   Roughness=ScalarParameter() {defaultValue=0.125, group=04_Special, parameterName=SpecialRoughness, sortPriority=29}
   UseRefraction=StaticBoolParameter() {group=04_Special, parameterName=Special_UseRefraction, sortPriority=30}
   RefractionValue=ScalarParameter() {defaultValue=1.011, group=04_Special, parameterName=SpecialRefractionValue, sortPriority=31}
   OverrideThinTranslucent=StaticBoolParameter() {group=04_Special, parameterName=Special_UseThinTranslucent, sortPriority=40}
   ThinTranslucentColor=@n24.RGB
   ThinTranslucentCoverage=ScalarParameter() {defaultValue=1, group=04_Special, parameterName=SpecialThinTranslucent_Coverage, sortPriority=42}
  ) {materialFunction=&MF_Glass_Lens_Section}
  n26 = VectorParameter() {
   desc=Alpha increases the intensity of the color
   group=04_Special_Enhancement
   parameterName=Special_RimColorXPower
   sortPriority=6
   defaultValue=<rgba 1 1 1 0.5>
  }
  n27 = VectorParameter() {
   desc=Alpha increases the intensity of the color
   group=04_Special_Enhancement
   parameterName=Special_SurfaceColorXPower
   sortPriority=8
   defaultValue=<rgba 1 1 1 0.5>
  }
  n28 = MaterialFunctionCall(
   MA=@n25.Result
   UseRimEnhancement=StaticBoolParameter() {group=04_Special_Enhancement, parameterName=Special_UseRimEnhancement, sortPriority=0}
   RimNormalDistance=ScalarParameter() {defaultValue=1.5, group=04_Special_Enhancement, parameterName=Special_RimNormalDistance, sortPriority=0}
   Rim_ColorXPower=AppendVector(A=@n26.RGB, B=@n26.RGB)
   UseSurfaceEnhancement=StaticBoolParameter() {group=04_Special_Enhancement, parameterName=Special_UseSurfaceEnhancement, sortPriority=3}
   SurfaceNormalDistance=ScalarParameter() {defaultValue=0.45, group=04_Special_Enhancement, parameterName=Special_SurfaceNormalDistance, sortPriority=2}
   Surface_ColorXPower=AppendVector(A=@n27.RGB, B=@n27.RGB)
   SurfaceGlow_Power=ScalarParameter() {defaultValue=50, group=04_Special_Enhancement, parameterName=Special_SurfaceGlowPower, sortPriority=4}
  ) {materialFunction=&MF_Lenses_LensEnhancement_v2}
  n29 = StaticSwitch(
   True=BlendMaterialAttributes(
    A=@n19
    B=StaticSwitchParameter(True=@n28.Result, False=@n25.Result) {group=04_Special, parameterName=Special_UseLensEnhancement, sortPriority=-10}
    Alpha=NamedRerouteUsage()
   )
   False=@n19
   Value=NamedRerouteUsage()
  )
  n30 = VectorParameter() {group=03_Reflector, parameterName=ReflectorColor, sortPriority=0, defaultValue=<rgba 0.698 0.3137 0 1>}
  n31 = ScalarParameter() {
   defaultValue=0.025
   desc=Controls the width of the inner thickness width
   group=03_Reflector
   parameterName=ReflectorThickness_Width
   sortPriority=6
  }
  n32 = ScalarParameter() {
   defaultValue=0.5
   desc=Controls the height of the inner thickness color
   group=03_Reflector
   parameterName=ReflectorThickness_Height
   sortPriority=7
  }
  n33 = VectorParameter() {group=03_Reflector, parameterName=ReflectorTint_LinearBurnColor, sortPriority=11, defaultValue=<rgba 0.001 0.001 0.001 1>}
  n34 = VectorParameter() {group=03_Reflector, parameterName=ReflectorThinTranslucent_Color, sortPriority=41, defaultValue=<rgba 0.7 0.7 0.7 1>}
  n35 = MaterialFunctionCall(
   In=@n29
   Color_Lens=@n30.RGB
   Fresnel_Exponent=ScalarParameter() {defaultValue=5, group=03_Reflector, parameterName=ReflectorFresnel_Exponent, sortPriority=1}
   Fresnel_Strength=ScalarParameter() {defaultValue=1, group=03_Reflector, parameterName=ReflectorFresnel_Strength, sortPriority=2}
   Fresnel_InnerTransparency=ScalarParameter() {defaultValue=0.375, group=03_Reflector, parameterName=ReflectorFresnel_InnerTransparency, sortPriority=3}
   Fresnel_OuterTransparency=ScalarParameter() {defaultValue=0.98, group=03_Reflector, parameterName=ReflectorFresnel_OuterTransparency, sortPriority=4}
   UseAdditionalThickness=StaticBoolParameter() {group=03_Reflector, parameterName=Reflector_UseAdditionalThickness, sortPriority=5}
   Thickness_Opacity=ScalarParameter() {defaultValue=1, group=03_Reflector, parameterName=ReflectorThickness_Opacity, sortPriority=5}
   Thickness_Width=@n31
   Thickness_Height=@n32
   UseTint=StaticBoolParameter() {group=03_Reflector, parameterName=Reflector_UseTint, sortPriority=10}
   Color_Tint=@n33.RGB
   Tint_BlendOpacity=ScalarParameter() {defaultValue=0.5, group=03_Reflector, parameterName=ReflectorTint_BlendAmount, sortPriority=12}
   OverridePBR=StaticBoolParameter() {group=03_Reflector, parameterName=Reflector_UsePBR, sortPriority=26}
   Metallic=ScalarParameter() {defaultValue=0.25, group=03_Reflector, parameterName=ReflectorMetallic, sortPriority=27}
   Specular=ScalarParameter() {defaultValue=0.8, group=03_Reflector, parameterName=ReflectorSpecular, sortPriority=28}
   Roughness=ScalarParameter() {defaultValue=0.125, group=03_Reflector, parameterName=ReflectorRoughness, sortPriority=29}
   UseRefraction=StaticBoolParameter() {group=03_Reflector, parameterName=Reflector_UseRefraction, sortPriority=30}
   RefractionValue=ScalarParameter() {defaultValue=1.011, group=03_Reflector, parameterName=ReflectorRefractionValue, sortPriority=31}
   OverrideThinTranslucent=StaticBoolParameter() {group=03_Reflector, parameterName=Reflector_UseThinTranslucent, sortPriority=40}
   ThinTranslucentColor=@n34.RGB
   ThinTranslucentCoverage=ScalarParameter() {defaultValue=1, group=03_Reflector, parameterName=ReflectorThinTranslucent_Coverage, sortPriority=42}
  ) {materialFunction=&MF_Glass_Lens_Section}
  n36 = VectorParameter() {
   desc=Alpha increases the intensity of the color
   group=03_Reflector_Enhancement
   parameterName=Reflector_RimColorXPower
   sortPriority=2
   defaultValue=<rgba 1 1 1 0.5>
  }
  n37 = VectorParameter() {
   desc=Alpha increases the intensity of the color
   group=03_Reflector_Enhancement
   parameterName=Reflector_SurfaceColorXPower
   sortPriority=5
   defaultValue=<rgba 1 1 1 0.5>
  }
  n38 = MaterialFunctionCall(
   MA=@n35.Result
   UseRimEnhancement=StaticBoolParameter() {group=03_Reflector_Enhancement, parameterName=Reflector_UseRimEnhancement, sortPriority=0}
   RimNormalDistance=ScalarParameter() {defaultValue=1.5, group=03_Reflector_Enhancement, parameterName=Reflector_RimNormalDistance, sortPriority=1}
   Rim_ColorXPower=AppendVector(A=@n36.RGB, B=@n36.RGB)
   UseSurfaceEnhancement=StaticBoolParameter() {group=03_Reflector_Enhancement, parameterName=Reflector_UseSurfaceEnhancement, sortPriority=3}
   SurfaceNormalDistance=ScalarParameter() {defaultValue=0.45, group=03_Reflector_Enhancement, parameterName=Reflector_SurfaceNormalDistance, sortPriority=4}
   Surface_ColorXPower=AppendVector(A=@n37.RGB, B=@n37.RGB)
   SurfaceGlow_Power=ScalarParameter() {defaultValue=50, group=03_Reflector_Enhancement, parameterName=Reflector_SurfaceGlowPower, sortPriority=6}
  ) {materialFunction=&MF_Lenses_LensEnhancement_v2}
  n39 = StaticSwitch(
   True=BlendMaterialAttributes(
    A=@n29
    B=StaticSwitchParameter(True=@n38.Result, False=@n35.Result) {group=03_Reflector, parameterName=Reflector_UseLensEnhancement, sortPriority=-10}
    Alpha=NamedRerouteUsage()
   )
   False=@n29
   Value=NamedRerouteUsage()
  )
  n40 = VectorParameter() {group=02_Rearlight, parameterName=RearlightColor, sortPriority=0, defaultValue=<rgba 0.4375 0 0 1>}
  n41 = ScalarParameter() {
   defaultValue=0.025
   desc=Controls the width of the inner thickness width
   group=02_Rearlight
   parameterName=RearlightThickness_Width
   sortPriority=6
  }
  n42 = ScalarParameter() {
   defaultValue=0.5
   desc=Controls the height of the inner thickness color
   group=02_Rearlight
   parameterName=RearlightThickness_Height
   sortPriority=7
  }
  n43 = VectorParameter() {group=02_Rearlight, parameterName=RearlightTint_LinearBurnColor, sortPriority=11, defaultValue=<rgba 0.001 0.001 0.001 1>}
  n44 = VectorParameter() {group=02_Rearlight, parameterName=RearlightThinTranslucent_Color, sortPriority=41, defaultValue=<rgba 0.7 0.7 0.7 1>}
  n45 = MaterialFunctionCall(
   In=@n39
   Color_Lens=@n40.RGB
   Fresnel_Exponent=ScalarParameter() {defaultValue=5, group=02_Rearlight, parameterName=RearlightFresnel_Exponent, sortPriority=1}
   Fresnel_Strength=ScalarParameter() {defaultValue=1, group=02_Rearlight, parameterName=RearlightFresnel_Strength, sortPriority=2}
   Fresnel_InnerTransparency=ScalarParameter() {defaultValue=0.375, group=02_Rearlight, parameterName=RearlightFresnel_InnerTransparency, sortPriority=3}
   Fresnel_OuterTransparency=ScalarParameter() {defaultValue=0.98, group=02_Rearlight, parameterName=RearlightFresnel_OuterTransparency, sortPriority=4}
   UseAdditionalThickness=StaticBoolParameter() {group=02_Rearlight, parameterName=Rearlight_UseAdditionalThickness, sortPriority=5}
   Thickness_Opacity=ScalarParameter() {defaultValue=1, group=02_Rearlight, parameterName=RearlightThickness_Opacity, sortPriority=5}
   Thickness_Width=@n41
   Thickness_Height=@n42
   UseTint=StaticBoolParameter() {group=02_Rearlight, parameterName=Rearlight_UseTint, sortPriority=10}
   Color_Tint=@n43.RGB
   Tint_BlendOpacity=ScalarParameter() {defaultValue=0.5, group=02_Rearlight, parameterName=RearlightTint_BlendAmount, sortPriority=12}
   OverridePBR=StaticBoolParameter() {group=02_Rearlight, parameterName=Rearlight_UsePBR, sortPriority=26}
   Metallic=ScalarParameter() {defaultValue=0.25, group=02_Rearlight, parameterName=RearlightMetallic, sortPriority=27}
   Specular=ScalarParameter() {defaultValue=0.8, group=02_Rearlight, parameterName=RearlightSpecular, sortPriority=28}
   Roughness=ScalarParameter() {defaultValue=0.125, group=02_Rearlight, parameterName=RearlightRoughness, sortPriority=29}
   UseRefraction=StaticBoolParameter() {group=02_Rearlight, parameterName=Rearlight_UseRefraction, sortPriority=30}
   RefractionValue=ScalarParameter() {defaultValue=1.011, group=02_Rearlight, parameterName=RearlightRefractionValue, sortPriority=31}
   OverrideThinTranslucent=StaticBoolParameter() {group=02_Rearlight, parameterName=Rearlight_UseThinTranslucent, sortPriority=40}
   ThinTranslucentColor=@n44.RGB
   ThinTranslucentCoverage=ScalarParameter() {defaultValue=1, group=02_Rearlight, parameterName=RearlightThinTranslucent_Coverage, sortPriority=42}
  ) {materialFunction=&MF_Glass_Lens_Section}
  n46 = VectorParameter() {
   desc=Alpha increases the intensity of the color
   group=02_Rearlight_Enhancement
   parameterName=Rearlight_RimColorXPower
   sortPriority=2
   defaultValue=<rgba 1 1 1 0.5>
  }
  n47 = VectorParameter() {
   desc=Alpha increases the intensity of the color
   group=02_Rearlight_Enhancement
   parameterName=Rearlight_SurfaceColorXPower
   sortPriority=5
   defaultValue=<rgba 1 1 1 0.5>
  }
  n48 = MaterialFunctionCall(
   MA=@n45.Result
   UseRimEnhancement=StaticBoolParameter() {group=02_Rearlight_Enhancement, parameterName=Rearlight_UseRimEnhancement, sortPriority=0}
   RimNormalDistance=ScalarParameter() {defaultValue=1.5, group=02_Rearlight_Enhancement, parameterName=Rearlight_RimNormalDistance, sortPriority=1}
   Rim_ColorXPower=AppendVector(A=@n46.RGB, B=@n46.RGB)
   UseSurfaceEnhancement=StaticBoolParameter() {group=02_Rearlight_Enhancement, parameterName=Rearlight_UseSurfaceEnhancement, sortPriority=3}
   SurfaceNormalDistance=ScalarParameter() {defaultValue=0.45, group=02_Rearlight_Enhancement, parameterName=Rearligh_SurfaceNormalDistance, sortPriority=4}
   Surface_ColorXPower=AppendVector(A=@n47.RGB, B=@n47.RGB)
   SurfaceGlow_Power=ScalarParameter() {defaultValue=50, group=02_Rearlight_Enhancement, parameterName=Rearlight_SurfaceGlowPower, sortPriority=6}
  ) {materialFunction=&MF_Lenses_LensEnhancement_v2}
  n49 = StaticSwitch(
   True=BlendMaterialAttributes(
    A=@n39
    B=StaticSwitchParameter(True=@n48.Result, False=@n45.Result) {group=02_Rearlight, parameterName=Rearlight_UseLensEnhancement, sortPriority=-10}
    Alpha=NamedRerouteUsage()
   )
   False=@n39
   Value=NamedRerouteUsage()
  )
  n50 = GetMaterialAttributes(None=@n49) {attributeGetTypes [#1]}
  n51 = MaterialFunctionCall!(Opacity=@n50.Opacity) {materialFunction=/Game/Global/Materials/MaterialGym/Materials/SortedOutline/MF_SortTranslucency}
  n52 = StaticBoolParameter() {parameterName=UseSortTranslucency}
  FunctionOutput(None=StaticSwitch(
    True=SetMaterialAttributes!(MaterialAttributes=@n50.MaterialAttributes, Opacity=@n51.Result) {attributeSetTypes [#1]}
    False=@n49
    Value=FunctionInput(Preview=@n52) {bUsePreviewValueAsDefault=true, inputName=UseSortTranslucency, inputType=FunctionInput_StaticBool}
   ))
  NamedRerouteDeclaration(None=MaterialFunctionCall!(UDimm_UVLocation=Constant2Vector()) {materialFunction=/VehicleCosmetics/Materials/ExtraFunctions/MF_UDimmMask}.Mask) {name=Misc_Mask, nodeColor=<rgba 0.4062 1 0 1>}
  NamedRerouteDeclaration(None=FunctionInput(Preview=StaticBool() {value=true}) {bUsePreviewValueAsDefault=true, inputName=UseHeadlightsAsShared, inputType=FunctionInput_StaticBool}) {name=Misc_Switch, nodeColor=<rgba 0 1 0.9531 1>}
  NamedRerouteDeclaration(None=FunctionInput(Preview=StaticBool() {value=true}) {bUsePreviewValueAsDefault=true, inputName=UseCustomRearlights, inputType=FunctionInput_StaticBool}) {name=Rearlight_Switch, nodeColor=<rgba 0 1 0.9531 1>}
  NamedRerouteDeclaration(None=FunctionInput(Preview=StaticBool()) {bUsePreviewValueAsDefault=true, inputName=UseCustomSpecial, inputType=FunctionInput_StaticBool}) {name=Special_Switch, nodeColor=<rgba 0.6875 0 1 1>}
  NamedRerouteDeclaration(None=FunctionInput(Preview=StaticBool()) {bUsePreviewValueAsDefault=true, inputName=UseCustomReflector, inputType=FunctionInput_StaticBool}) {name=Reflector_Switch, nodeColor=<rgba 0 1 0.3672 1>}
  NamedRerouteDeclaration(None=MaterialFunctionCall(UDimm_UVLocation=Constant2Vector() {r=-1}) {materialFunction=&MF_UDimmMask}.Mask) {name=Rearlight_Mask, nodeColor=<rgba 1 0 0.2578 1>}
  NamedRerouteDeclaration(None=MaterialFunctionCall(UDimm_UVLocation=Constant2Vector() {g=1}) {materialFunction=&MF_UDimmMask}.Mask) {name=Reflector_Mask, nodeColor=<rgba 0.5 1 0 1>}
  NamedRerouteDeclaration(None=MaterialFunctionCall(UDimm_UVLocation=Constant2Vector() {g=1, r=-1}) {materialFunction=&MF_UDimmMask}.Mask) {name=Special_Mask, nodeColor=<rgba 0 1 0.6953 1>}
 ]}
