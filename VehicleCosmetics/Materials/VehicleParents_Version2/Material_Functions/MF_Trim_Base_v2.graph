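# Vehicle trim material function that outputs material attributes for the trim layer of cosmetic vehicle materials.
# Static switches select texture or parameter base color, metallic and roughness, and optional emissive from a remapped mask times color and brightness.
# A tiled noise normal, flat at low quality, feeds the surface.
# Optional clear coat adds a flattenable bottom normal, with a composite blending Default Lit and Clear Coat by the rounded-up coat amount.
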MaterialFunction /VehicleCosmetics/Materials/VehicleParents_Version2/Material_Functions/MF_Trim_Base_v2
graph {nodes [
  n1 = TextureSampleParameter2D() {
   desc=_M
   group=Textures
   parameterName=Masks
   samplerSource=SSM_Wrap_WorldGroupSettings
   samplerType=SAMPLERTYPE_LinearColor
   sortPriority=33
   texture=/VehicleCosmetics/Textures/T_Black_Linear
   channelNames {a="", b=BoostMask, g=PaintMask, r=EmssivePartsMask}
  }
  n2 = TextureSampleParameter2D() {
   group=Texture
   parameterName=BodyTextures
   samplerSource=SSM_Wrap_WorldGroupSettings
   samplerType=SAMPLERTYPE_LinearColor
   texture=&T_Black_Linear
  }
  n3 = StaticSwitchParameter(True=@n1.EmssivePartsMask, False=@n2.B) {parameterName=ApplyTrim}
  n4 = Saturate(None=Multiply(A=Subtract(A=@n3) {constB=0.098}) {constB=1.1087})
  n5 = Multiply(
   A=VectorParameter() {group=Trim, parameterName=Trim_EmissiveColor, sortPriority=50, defaultValue=<rgba 0 1 1 1>}.RGB
   B=ScalarParameter() {defaultValue=1, group=Trim, parameterName=Trim_EmissiveBrightness, sortPriority=51}
  )
  n6 = SetMaterialAttributes!(
   MaterialAttributes
   BaseColor=StaticSwitchParameter(True=NamedRerouteUsage(), False=VectorParameter() {group=Trim, parameterName=TrimColor, defaultValue=<rgba 0.012 0.012 0.012 1>}.RGB) {group=Trim, parameterName=Trim_UseBaseColorTexture}
   Metallic=StaticSwitchParameter(True=NamedRerouteUsage(), False=ScalarParameter() {group=Trim, parameterName=Trim_Metallic}) {group=Trim, parameterName=Trim_UseBodyTexturesPBR}
   Roughness=StaticSwitchParameter(True=NamedRerouteUsage(), False=ScalarParameter() {defaultValue=0.45, group=Trim, parameterName=Trim_Roughness}) {group=Trim, parameterName=Trim_UseBodyTexturesPBR}
   EmissiveColor=StaticSwitchParameter(True=Multiply(A=@n4, B=@n5), False=Constant()) {group=Trim, parameterName=ApplyTrimEmissive}
   Normal=NamedRerouteUsage()
   ShadingModel=ShadingModel()
  ) {attributeSetTypes [
    #1
    #2
    #3
    #4
    #5
    #6
   ]}
  n7 = Saturate(None=ScalarParameter() {defaultValue=1, group=Trim, parameterName=Trim_ClearCoat})
  n8 = TextureSampleParameter2D() {group=Finish, parameterName=TrimBottomNormalBase, samplerType=SAMPLERTYPE_Normal, texture=/VehicleCosmetics/Textures/T_BlankNormal_B}
  n9 = ScalarParameter() {group=DetailNormal, parameterName=TrimBottomNormalFlatness}
  n10 = SetMaterialAttributes!(
   MaterialAttributes=@n6
   CustomData0=@n7
   CustomData1=Saturate(None=ScalarParameter() {group=Trim, parameterName=Trim_ClearCoatRoughness})
   ClearCoatBottomNormal=NamedRerouteDeclaration(None=MaterialFunctionCall(Normal=@n8.RGB, Flatness=@n9) {materialFunction=&FlattenNormal}.Result) {name=CCBottomNormal, nodeColor=<rgba 1 0.6562 0 1>}
   ShadingModel=ShadingModel() {shadingModel=MSM_ClearCoat}
  ) {attributeSetTypes [
    #7
    #8
    #9
    #6
   ]}
  Composite!("In (MA)"=@n10, CCThickness=@n7) {subgraphName=BlendShadingModel}
  MaterialExpressionReroute_61 = Reroute(None=@n10)
  n11 = GetMaterialAttributes(None=@MaterialExpressionReroute_61."In (MA)") {attributeGetTypes [#5, #9]}
  MaterialExpressionReroute_2 = Reroute(None=@n7)
  n12 = BlendMaterialAttributes(
   A=SetMaterialAttributes!(
    MaterialAttributes=@n11.MaterialAttributes
    ShadingModel=ShadingModel()
    Normal=MaterialFunctionCall(BaseNormal=@n11.Normal, AdditionalNormal=@n11.Normal) {materialFunction=&BlendAngleCorrectedNormals}.Result
   ) {attributeSetTypes [#6, #5]}
   B=@MaterialExpressionReroute_61."In (MA)"
   Alpha=Saturate(None=Ceil(None=@MaterialExpressionReroute_2.CCThickness))
  )
  MaterialExpressionReroute_63 = Reroute(None=@n12)
  FunctionOutput(None=StaticSwitchParameter(True=@MaterialExpressionReroute_63.Output, False=@n6) {group=Trim, parameterName=Trim_UseClearCoatShading})
  MaterialFunctionCall() {materialFunction=&PlatformSwitch_1D}
  MaterialFunctionCall!("Xb1,PS4", "PC-High&&Epic,XSX,PS5", "PC-Perf,Android", Switch, "PC-Low(NotPerf)&&Medium") {materialFunction=/Game/Effects/Niagara/MasterMaterials/Functions/Common/PlatformSwitch_2D}
  n13 = TextureSampleParameter2D(UVs=Multiply(A=TextureCoordinate(), B=ScalarParameter() {defaultValue=64, group=Trim, parameterName=Trim_NoiseTiling})) {
   group=Textures
   parameterName=Normal
   samplerSource=SSM_Wrap_WorldGroupSettings
   samplerType=SAMPLERTYPE_Normal
   sortPriority=34
   texture=/VehicleCosmetics/Textures/UniversalTextures/NormalNoise_universal
  }
  NamedRerouteDeclaration(None=QualitySwitch(Default=@n13.RGB, Low=Constant3Vector() {constant=<rgba 0 0 1 0>})) {name=NormalTexture, nodeColor=<rgba 0.6038 0.3325 0.013 1>}
  n14 = TextureSampleParameter2D() {
   group=Textures
   parameterName=BaseColor
   samplerSource=SSM_Wrap_WorldGroupSettings
   sortPriority=30
   texture=/VehicleCosmetics/Textures/T_Black
  }
  NamedRerouteDeclaration(None=@n14.RGB) {name=BaseColorTexture, nodeColor=<rgba 0.6038 0.3325 0.013 1>}
  n15 = TextureSampleParameter2D() {
   group=Trim
   parameterName=BodyTextures
   samplerSource=SSM_Wrap_WorldGroupSettings
   samplerType=SAMPLERTYPE_LinearColor
   sortPriority=35
   texture=&T_Black_Linear
   channelNames {a=Roughness, b=Metallic, g=AO-BlackOut, r=Curvature}
  }
  NamedRerouteDeclaration(None=@n15.Roughness) {name=Roughness, nodeColor=<rgba 0 0.1726 0.4792 1>}
  NamedRerouteDeclaration(None=@n15.Metallic) {name=Metallic, nodeColor=<rgba 0 0.1726 0.4792 1>}
  PinBase() {reroutePins [
    "In (MA)" {expression=&self:MaterialExpressionReroute_61}
    CCThickness {expression=&self:MaterialExpressionReroute_2}
   ]}
  PinBase!(Output=@n12) {reroutePins [Output {expression=&self:MaterialExpressionReroute_63}]}
 ]}
